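# Post-process AnimBlueprint for the HydraTrumpet CoachSalve female head that adds secondary motion to its hair/fur bones.
# The input pose passes through seven chained Anim Dynamics nodes, each simulating one front, left or right dyn_hairFur bone.
# Each uses a stiff angular spring, tight two-axis angular limits, overridden damping and scaled-down gravity in root-relative space.
# The event graph holds only unconnected stub nodes, with no logic or variables.
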AnimBlueprint /BRCosmetics/Characters/Player/Female/Medium/Heads/F_MED_HydraTrumpet_CoachSalve_Head/Meshes/CustomSkel/H_HydraTrumpet_CoachS_Post_AnimBP
target_skeleton=&/Head_HydraTrumpet_CoachSalve_Skel
graphs [
 AnimGraph {nodes [
   n1 = Anim Dynamics - Bone: dyn_hairFur_01_F(ComponentPose=Local To Component(LocalPose=Input Pose()), Alpha=1, ExternalForce="0,0,0") {
    angularDampingOverride=1
    angularSpringConstant=25000
    bAngularSpring=true
    bOverrideAngularDamping=true
    bUseGravityOverride=true
    gravityScale=0.35
    simulationSpace=RootRelative
    alphaBoolBlend {blendOption=Linear}
    boundBone {boneName=dyn_hairFur_01_F}
    componentLinearAccScale=<1 1 1>
    componentLinearVelScale=<1 1 1>
    gravityOverride {z=-980}
    physicsBodyDefinitions [{
      boundBone {boneName=dyn_hairFur_01_F}
      boxExtents=<30 30 30>
      constraintSetup {angularLimitsMax=<0 15 15>, angularLimitsMin=<0 -15 -15>, angularTarget=<1 0 0>}
      localJointOffset=<1 0 0>
     }]
    relativeSpaceBone {boneName=head}
   }
   n2 = Anim Dynamics - Bone: dyn_hairFur_01_L(ComponentPose=@n1, Alpha=1, ExternalForce="0,0,0") {
    angularDampingOverride=1
    angularSpringConstant=25000
    bAngularSpring=true
    bOverrideAngularDamping=true
    bUseGravityOverride=true
    gravityScale=0.35
    simulationSpace=RootRelative
    alphaBoolBlend {blendOption=Linear}
    boundBone {boneName=dyn_hairFur_01_L}
    componentLinearAccScale=<1 1 1>
    componentLinearVelScale=<1 1 1>
    gravityOverride {z=-980}
    physicsBodyDefinitions [{
      boundBone {boneName=dyn_hairFur_01_L}
      boxExtents=<30 30 30>
      constraintSetup {angularLimitsMax=<0 15 15>, angularLimitsMin=<0 -15 -15>, angularTarget=<1 0 0>}
      localJointOffset=<3 0 0>
     }]
    relativeSpaceBone {boneName=head}
   }
   n3 = Anim Dynamics - Bone: dyn_hairFur_02_L(ComponentPose=@n2, Alpha=1, ExternalForce="0,0,0") {
    angularDampingOverride=1
    angularSpringConstant=25000
    bAngularSpring=true
    bOverrideAngularDamping=true
    bUseGravityOverride=true
    gravityScale=0.35
    simulationSpace=RootRelative
    alphaBoolBlend {blendOption=Linear}
    boundBone {boneName=dyn_hairFur_02_L}
    componentLinearAccScale=<1 1 1>
    componentLinearVelScale=<1 1 1>
    gravityOverride {z=-980}
    physicsBodyDefinitions [{
      boundBone {boneName=dyn_hairFur_02_L}
      boxExtents=<25 25 25>
      constraintSetup {angularLimitsMax=<0 15 15>, angularLimitsMin=<0 -15 -15>, angularTarget=<1 0 0>}
      localJointOffset=<3 0 0>
     }]
    relativeSpaceBone {boneName=head}
   }
   n4 = Anim Dynamics - Bone: dyn_hairFur_01_R(ComponentPose=@n3, Alpha=1, ExternalForce="0,0,0") {
    angularDampingOverride=1
    angularSpringConstant=25000
    bAngularSpring=true
    bOverrideAngularDamping=true
    bUseGravityOverride=true
    gravityScale=0.35
    simulationSpace=RootRelative
    alphaBoolBlend {blendOption=Linear}
    boundBone {boneName=dyn_hairFur_01_R}
    componentLinearAccScale=<1 1 1>
    componentLinearVelScale=<1 1 1>
    gravityOverride {z=-980}
    physicsBodyDefinitions [{
      boundBone {boneName=dyn_hairFur_01_R}
      boxExtents=<30 30 30>
      constraintSetup {angularLimitsMax=<0 15 15>, angularLimitsMin=<0 -15 -15>, angularTarget=<1 0 0>}
      localJointOffset=<3 0 0>
     }]
    relativeSpaceBone {boneName=head}
   }
   n5 = Anim Dynamics - Bone: dyn_hairFur_02_R(ComponentPose=@n4, Alpha=1, ExternalForce="0,0,0") {
    angularDampingOverride=1
    angularSpringConstant=25000
    bAngularSpring=true
    bOverrideAngularDamping=true
    bUseGravityOverride=true
    gravityScale=0.35
    simulationSpace=RootRelative
    alphaBoolBlend {blendOption=Linear}
    boundBone {boneName=dyn_hairFur_02_R}
    componentLinearAccScale=<1 1 1>
    componentLinearVelScale=<1 1 1>
    gravityOverride {z=-980}
    physicsBodyDefinitions [{
      boundBone {boneName=dyn_hairFur_02_R}
      boxExtents=<25 25 25>
      constraintSetup {angularLimitsMax=<0 15 15>, angularLimitsMin=<0 -15 -15>, angularTarget=<1 0 0>}
      localJointOffset=<3 0 0>
     }]
    relativeSpaceBone {boneName=head}
   }
   n6 = Anim Dynamics - Bone: dyn_hairFur_02_F(ComponentPose=@n5, Alpha=1, ExternalForce="0,0,0") {
    angularDampingOverride=1
    angularSpringConstant=30000
    bAngularSpring=true
    bOverrideAngularDamping=true
    bUseGravityOverride=true
    gravityScale=0.35
    simulationSpace=RootRelative
    alphaBoolBlend {blendOption=Linear}
    boundBone {boneName=dyn_hairFur_02_F}
    componentLinearAccScale=<1 1 1>
    componentLinearVelScale=<1 1 1>
    gravityOverride {z=-980}
    physicsBodyDefinitions [{
      boundBone {boneName=dyn_hairFur_02_F}
      boxExtents=<30 30 30>
      constraintSetup {angularLimitsMax=<0 15 15>, angularLimitsMin=<0 -15 -15>, angularTarget=<1 0 0>}
      localJointOffset=<1 0 0>
     }]
    relativeSpaceBone {boneName=head}
   }
   n7 = Anim Dynamics - Bone: dyn_hairFur_03_F(ComponentPose=@n6, Alpha=1, ExternalForce="0,0,0") {
    angularDampingOverride=1
    angularSpringConstant=30000
    bAngularSpring=true
    bOverrideAngularDamping=true
    bUseGravityOverride=true
    gravityScale=0.35
    simulationSpace=RootRelative
    alphaBoolBlend {blendOption=Linear}
    boundBone {boneName=dyn_hairFur_03_F}
    componentLinearAccScale=<1 1 1>
    componentLinearVelScale=<1 1 1>
    gravityOverride {z=-980}
    physicsBodyDefinitions [{
      boundBone {boneName=dyn_hairFur_03_F}
      boxExtents=<30 30 30>
      constraintSetup {angularLimitsMax=<0 15 15>, angularLimitsMin=<0 -15 -15>, angularTarget=<1 0 0>}
      localJointOffset=<1 0 0>
     }]
    relativeSpaceBone {boneName=head}
   }
   Output Pose(Result=Component To Local(ComponentPose=@n7))
  ]}
 EventGraph {nodes [
   TryGetPawnOwner()
   Event BlueprintUpdateAnimation()
  ]}
]
variables []
